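; Targeting task for the AmberVine force power's water area, overriding the source location and direction of a trace.
; The trace origin comes from the water area origin on the source actor's AmberVine pawn component.
; The direction is the forward vector of the source Fort player pawn's control rotation.
; Sequence nodes fall through to the parent implementation when the subsystem, or the component for the origin, is invalid.
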(fn GetSourceLocation
  (params (TargetingHandle :type Targeting%20Request%20Handle%20Structure%20%28by%20ref%29))
  (bind _returnvalue (Targeting|GetTargetingSubsystem TargetingHandle))
  (bind _returnvalue_2 (TargetingSystem|TargetingTypes|GetTargetingSourceContext _returnvalue TargetingHandle))
  (bind _sourceactor (Utilities|Struct|BreakTargetingSourceContext _returnvalue_2))
  (bind _returnvalue_3 (Actor|GetComponentbyClass _sourceactor "/PlotTwist/ForcePowers/AmberVine/BP_AmbervinePawnComponent.BP_AmbervinePawnComponent_C"))
  ; Get the location from which the trace is done.
  (Utilities|FlowControl|Sequence
    (:then_0
      (Utilities|IsValid _returnvalue
        (:Is%20Valid
          (Utilities|IsValid _returnvalue_3
            (:Is%20Valid
              (return (Variables|/PlotTwist/ForcePowers/AmberVine/BP_AmbervinePawnComponent.BP_AmbervinePawnComponent_C|GetWaterAreaOrigin _returnvalue_3)))))))
    (:then_1
      (bind _returnvalue_4 (|CallParentFunction TargetingHandle))
      (return _returnvalue_4))))

(fn GetTraceDirection
  (params (TargetingHandle :type Targeting%20Request%20Handle%20Structure%20%28by%20ref%29))
  (bind _returnvalue (Targeting|GetTargetingSubsystem TargetingHandle))
  (Utilities|FlowControl|Sequence
    (:then_0
      (Utilities|IsValid _returnvalue
        (:Is%20Valid
          (bind _asfort_player_pawn (Utilities|Casting|CastToFortPlayerPawn (Utilities|Struct|BreakTargetingSourceContext (TargetingSystem|TargetingTypes|GetTargetingSourceContext _returnvalue TargetingHandle))))
          (return (Math|Vector|GetForwardVector (Pawn|GetControlRotation _asfort_player_pawn :!class /Script/Engine.Pawn))))))
    (:then_1
      (bind _returnvalue_2 (|CallParentFunction TargetingHandle))
      (return _returnvalue_2))))
